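# Sound cue for the TameFoot NPC boss voice line played, by its name, on entering the DBNO (down but not out) state.
# A Random node picks one of three equally weighted DBNO voice waves and is excluded from branch culling.
# Output uses the NPC VO sound class, TameFoot close-range attenuation and a shared TameFoot chatter concurrency set.
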SoundCue /TameFoot/Sounds/VO/TameFoot_DBNO_Enter_Cue
root=Random_0
nodes [
 {id=Random_0, children [WavePlayer_1,WavePlayer_2,WavePlayer_3], properties {bShouldExcludeFromBranchCulling=true, weights [1,1,1]}}
 {id=WavePlayer_1, properties {soundWaveAssetPtr=/TameFoot/Sounds/VO/Waves/TameFoot_NPCBoss_DBNO_010}}
 {id=WavePlayer_2, properties {soundWaveAssetPtr=/TameFoot/Sounds/VO/Waves/TameFoot_NPCBoss_DBNO_012}}
 {id=WavePlayer_3, properties {soundWaveAssetPtr=/TameFoot/Sounds/VO/Waves/TameFoot_NPCBoss_DBNO_011}}
]
properties {
 attenuationSettings=&/TameFoot_Close_Att
 maxDistance=4000
 soundClassObject=&Fort_VO_NPC
 volumeMultiplier=1
 concurrencySet [&/TameFoot_Chatter_Conc]
}
