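# Sound cue for TameFoot NPC boss voice lines, played, by its name, when the boss loses its target while on full alert.
# A Random node picks one of five equally weighted LostTarget waves, and that node is excluded from branch culling.
# Playback uses the NPC VO sound class, close-range TameFoot attenuation and a chatter concurrency set that governs overlapping barks.
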SoundCue /TameFoot/Sounds/VO/TameFoot_FullAlertLostTarget_Cue
root=Random_2
nodes [
 {
  id=Random_2
  children [WavePlayer_15, WavePlayer_16, WavePlayer_17, WavePlayer_18, WavePlayer_19]
  properties {bShouldExcludeFromBranchCulling=true, weights [1,1,1,1,1]}
 }
 {id=WavePlayer_15, properties {soundWaveAssetPtr=/TameFoot/Sounds/VO/Waves/TameFoot_NPCBoss_LostTarget_010}}
 {id=WavePlayer_16, properties {soundWaveAssetPtr=/TameFoot/Sounds/VO/Waves/TameFoot_NPCBoss_LostTarget_011}}
 {id=WavePlayer_17, properties {soundWaveAssetPtr=/TameFoot/Sounds/VO/Waves/TameFoot_NPCBoss_LostTarget_012}}
 {id=WavePlayer_18, properties {soundWaveAssetPtr=/TameFoot/Sounds/VO/Waves/TameFoot_NPCBoss_LostTarget_013}}
 {id=WavePlayer_19, properties {soundWaveAssetPtr=/TameFoot/Sounds/VO/Waves/TameFoot_NPCBoss_LostTarget_014}}
]
properties {
 attenuationSettings=&/TameFoot_Close_Att
 maxDistance=4000
 soundClassObject=&Fort_VO_NPC
 volumeMultiplier=1
 concurrencySet [&/TameFoot_Chatter_Conc]
}
