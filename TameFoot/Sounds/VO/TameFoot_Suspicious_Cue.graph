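# Sound cue for the TameFoot NPC boss's suspicious voice barks.
# A Random node picks one of four equally weighted voice waves, and is excluded from branch culling.
# It plays on the NPC VO sound class with close-range attenuation and a slight volume boost, limited by the TameFoot chatter concurrency set.
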SoundCue /TameFoot/Sounds/VO/TameFoot_Suspicious_Cue
root=Random_1
nodes [
 {
  id=Random_1
  children [WavePlayer_1,WavePlayer_2,WavePlayer_6,WavePlayer_7]
  properties {bShouldExcludeFromBranchCulling=true, weights [1,1,1,1]}
 }
 {id=WavePlayer_1, properties {soundWaveAssetPtr=/TameFoot/Sounds/VO/Waves/TameFoot_NPCBoss_Suspicious_010}}
 {id=WavePlayer_2, properties {soundWaveAssetPtr=/TameFoot/Sounds/VO/Waves/TameFoot_NPCBoss_Suspicious_011}}
 {id=WavePlayer_6, properties {soundWaveAssetPtr=/TameFoot/Sounds/VO/Waves/TameFoot_NPCBoss_Suspicious_012}}
 {id=WavePlayer_7, properties {soundWaveAssetPtr=/TameFoot/Sounds/VO/Waves/TameFoot_NPCBoss_Suspicious_013}}
]
properties {
 attenuationSettings=&/TameFoot_Close_Att
 maxDistance=4000
 soundClassObject=&Fort_VO_NPC
 volumeMultiplier=1.2
 concurrencySet [&/TameFoot_Chatter_Conc]
}
